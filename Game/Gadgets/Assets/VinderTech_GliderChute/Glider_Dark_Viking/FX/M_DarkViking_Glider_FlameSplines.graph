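# Unlit, translucent, two-sided mesh-particle material that renders flame splines for the Dark Viking glider.
# Blended panning normals and thickness masks shape SplineThicken width, while a gradient-weighted BendOffset bends the vertex offset.
# Two panning noise textures, offset by vertex color, mix two colors remapped to HDR, optionally switched to particle color.
# Ramp and gradient textures shape opacity, with noise-driven depth fade and particle alpha.
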Material /Game/Gadgets/Assets/VinderTech_GliderChute/Glider_Dark_Viking/FX/M_DarkViking_Glider_FlameSplines
graph {
 properties {
  bUsedWithMeshParticles=true
  bUsedWithNiagaraMeshParticles=true
  blendMode=BLEND_Translucent
  pixelDepthOffsetMode=PDOM_Legacy
  shadingModel=MSM_Unlit
  twoSided=true
  usedShadingModels=Unlit
 }
 nodes [
  n1 = Multiply(A=TextureCoordinate() {vTiling=4}, B=ScalarParameter() {defaultValue=0.7, parameterName=NormalScale})
  n2 = Multiply(A=TextureCoordinate() {vTiling=4}, B=ScalarParameter() {defaultValue=1.5, parameterName=DetailNormalScale})
  n3 = MaterialFunctionCall(
   BaseNormal=TextureSample(UVs=Panner(Coordinate=@n1) {speedY=-2}) {samplerType=SAMPLERTYPE_Normal, texture=/Game/Global/Textures/Default/Tileable_Patterns/T_Water2_N}.RGB
   AdditionalNormal=TextureSample(UVs=Panner(Coordinate=@n2) {speedY=-1.3}) {samplerType=SAMPLERTYPE_Normal, texture=/Game/Effects/Fort_Effects/Textures/WaterFoam0001_19_N}.RGB
  ) {materialFunction=&BlendAngleCorrectedNormals}
  n4 = LinearInterpolate(Alpha=ComponentMask(None=@n3.Result) {r=true}) {constA=0.86}
  n5 = TextureCoordinate() {vTiling=3.1233}
  n6 = LinearInterpolate(Alpha=TextureSample(UVs=Panner(Coordinate=@n5) {speedY=0.25}) {texture=/Game/Effects/Fort_Effects/Textures/Masks/T_SplineThickness_Mask}.G) {constA=0.8}
  n7 = Add(A=PerInstanceRandom(), B=TextureCoordinate() {vTiling=2})
  n8 = LinearInterpolate(Alpha=TextureSample(UVs=Panner(Coordinate=@n7) {speedY=2}) {texture=&T_SplineThickness_Mask}.G) {constA=0.3}
  n9 = TextureCoordinate()
  n10 = Multiply(A=Multiply(A=Clamp(None=@n4), B=Multiply(A=@n6, B=@n8)), B=OneMinus(None=ComponentMask(None=@n9) {g=true}))
  n11 = StaticBool() {value=true}
  n12 = OneMinus(None=MaterialFunctionCall() {materialFunction=&LinearGradient}.VGradient)
  Add(
   A=MaterialFunctionCall(
    WidthBase=ScalarParameter() {defaultValue=320, parameterName=WidthBase}
    WidthTip=ScalarParameter() {defaultValue=18, parameterName=WidthTip}
    UVs for Thickness=Clamp(None=@n10)
    UVs for texturing=TextureCoordinate()
    Expand U or V UV Channel=@n11
    AngleCorrectedNormals=@n11
    AdditionalNormal=@n3.Result
   ) {materialFunction=&SplineThicken}.WorldPositionOffset
   B=Multiply(A=VectorParameter() {parameterName=BendOffset}.RGB, B=Power(Base=@n12) {constExponent=1.8})
  )
  n13 = VectorParameter() {parameterName=ColorA, defaultValue=<rgba 0 0.0367 0.075 1>}
  n14 = VectorParameter() {parameterName=ColorB, defaultValue=<rgba 0 0.9649 2 1>}
  n15 = MaterialFunctionCall(Float2=TextureCoordinate() {uTiling=0.25, vTiling=0.745}) {materialFunction=&BreakOutFloat2Components}
  n16 = VertexColor()
  n17 = ScalarParameter() {defaultValue=10, parameterName=RadnomAmount}
  n18 = Multiply(A=@n16.R, B=@n17)
  n19 = Add(A=@n15.R, B=@n18)
  n20 = Add(A=@n15.G, B=@n18)
  n21 = Multiply(A=Time(), B=ScalarParameter() {defaultValue=1.5, parameterName=Speed})
  n22 = TextureSample(UVs=Panner(Coordinate=AppendVector(A=@n19, B=@n20), Time=@n21) {speedX=0.05, speedY=1}) {texture=/Game/Global/Textures/Default/Caustics/T_Turbulence_Seamless}
  n23 = MaterialFunctionCall(Float2=TextureCoordinate() {uTiling=0.1}) {materialFunction=&BreakOutFloat2Components}
  n24 = Multiply(A=@n16.R, B=@n17)
  n25 = Add(A=@n23.R, B=@n24)
  n26 = Add(A=@n23.G, B=@n24)
  n27 = TextureSample(UVs=Panner(Coordinate=AppendVector(A=@n25, B=@n26), Time=@n21) {speedX=0.05, speedY=1}) {texture=&T_Inky_Smoke_Tile}
  n28 = Clamp(None=Multiply(A=Multiply(A=@n22.B, B=@n27.R)) {constB=10})
  n29 = ParticleColor()
  n30 = Multiply(A=@n29.RGB) {constB=0.5}
  n31 = ScalarParameter() {parameterName=Use Particle Color}
  n32 = LinearInterpolate(
   A=Clamp(None=LinearInterpolate(A=@n13.RGB, B=@n14.RGB, Alpha=@n28))
   B=Clamp(None=LinearInterpolate(A=@n29.RGB, B=@n30, Alpha=@n28))
   Alpha=@n31
  )
  n33 = LinearInterpolate(A=VectorParameter() {parameterName=HDR_Tint, defaultValue=<rgba 5500 25 25 0>}.RGB, B=Multiply(A=@n29.RGB) {constB=100}, Alpha=@n31)
  n34 = ScalarParameter() {defaultValue=1, parameterName=Alpha}
  MaterialFunctionCall(Emissive=Multiply(
    A=MaterialFunctionCall(LDR Input=@n32, HDR Tint=@n33) {materialFunction=&DeriveHDRfromLDR}.RemappedRange
    B=@n34
   )) {materialFunction=&MF_FX_EmissiveAdjustments}
  n35 = TextureCoordinate()
  n36 = Multiply(
   A=Multiply(A=@n34) {constB=0.75}
   B=Power(Base=TextureSample(UVs=@n35) {texture=/Game/Effects/Fort_Effects/Textures/Ramps/T_Ramp_Limbo_Pool_Gas}.G)
  )
  n37 = Power(Base=OneMinus(None=TextureSample() {texture=/Game/Effects/Fort_Effects/Effects/Characters/Generic/ResIn/SquareGradient}.R)) {constExponent=4}
  n38 = ScalarParameter() {defaultValue=50, parameterName=Top}
  n39 = ScalarParameter() {defaultValue=5, parameterName=Bottom}
  Multiply(
   A=ParticleColor().A
   B=DepthFade(Opacity=Multiply(A=@n36, B=@n37), FadeDistance=LinearInterpolate(A=@n38, B=@n39, Alpha=@n28) {constA=50, constB=10}) {fadeDistanceDefault=50}
  )
 ]
}
